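# Translucent UI master material for band member avatar icons in Sparks, with selectable square, circle or diamond shapes.
# It blends a PlayerIcon texture over a BgColor fill with an optional overlay-blended gradient, and an Inactive switch desaturates the result.
# SDF box and circle functions, with the box rotated for the diamond, build masks and tinted outlines.
# Static switches select the mask shape and border style.
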Material /SparksUICommon/Materials/M_UI_BandMember_Avatar_Master
graph {
 properties {blendMode=BLEND_Translucent, materialDomain=MD_UI, pixelDepthOffsetMode=PDOM_Legacy, usedShadingModels=Default Lit}
 nodes [
  n1 = NamedRerouteUsage()
  n2 = NamedRerouteUsage()
  Min(A=@n1, B=@n2)
  NamedRerouteDeclaration(None=Multiply(A=@n1, B=@n2)) {name=CircleWithoutBorder, nodeColor=<rgba 1 0.3516 0 1>}
  NamedRerouteDeclaration(None=MaterialFunctionCall(Size=Constant() {r=1}) {materialFunction=&MF_UI_Circle}.Fill) {name=CircleMask, nodeColor=<rgba 0.8516 1 0 1>}
  n3 = NamedRerouteUsage()
  n4 = Constant() {r=0.96}
  n5 = ScalarParameter() {defaultValue=0.025, parameterName=Circle Stroke}
  n6 = OneMinus(None=SmoothStep(
    Max=Constant() {r=0.03}
    Value=MaterialFunctionCall(Size=@n4, OutlineThickness=@n5) {materialFunction=&MF_UI_Circle}.SDFOutline
   ))
  n7 = NamedRerouteUsage()
  n8 = ComponentMask(None=Subtract(A=NamedRerouteUsage(), B=@n6)) {r=true}
  NamedRerouteDeclaration(None=Add(A=LinearInterpolate(B=@n3, Alpha=@n6), B=Multiply(A=@n7, B=@n8))) {name=CircleBorder, nodeColor=<rgba 0 0 0 1>}
  NamedRerouteDeclaration(None=MaterialFunctionCall(Size=Constant() {r=1}, CornerRoundness=ScalarParameter() {parameterName=SquareBorderRoudness}) {materialFunction=&MF_UI_Box}.Fill) {name=SquareAlpha, nodeColor=<rgba 0 0 0 1>}
  n9 = NamedRerouteUsage()
  n10 = Saturate(None=MaterialFunctionCall(Size=Constant() {r=1}, OutlineThickness=ScalarParameter() {defaultValue=0.1, parameterName=Square Stroke}) {materialFunction=&MF_UI_Box}.Outline)
  n11 = NamedRerouteUsage()
  n12 = ComponentMask(None=Subtract(A=NamedRerouteUsage(), B=@n10)) {r=true}
  NamedRerouteDeclaration(None=Add(A=LinearInterpolate(B=@n9, Alpha=@n10), B=Multiply(A=@n11, B=@n12))) {name=SquareBorder, nodeColor=<rgba 0 0 0 1>}
  NamedRerouteDeclaration(None=VectorParameter() {parameterName=BorderColor, defaultValue=<rgba 1 1 1 1>}.RGB) {name=BorderColor, nodeColor=<rgba 0 0 0 1>}
  n13 = MaterialFunctionCall(
   UVs=MaterialFunctionCall(UVs=TextureCoordinate(), "Rotation Angle (0-1)"=Constant() {r=0.125}) {materialFunction=&CustomRotator}.Rotated Values
   Size=Constant() {r=0.68}
  ) {materialFunction=&MF_UI_Box}
  n14 = Constant() {r=0.001}
  NamedRerouteDeclaration(None=Saturate(None=MaterialFunctionCall(DistanceField=@n13.SDF, Threshold=@n14) {materialFunction=&MF_UI_AntiAliasing}.Inner)) {name=DiamondAlpha, nodeColor=<rgba 0 0 0 1>}
  n15 = VectorParameter() {group=Color, parameterName=BgColor, defaultValue=<rgba 0.491 0.7011 0.0908 1>}
  n16 = MaterialFunctionCall() {materialFunction=&LinearGradient}
  n17 = MaterialFunctionCall(Base=LinearInterpolate(A=@n15.RGB, B=Divide(A=@n15.RGB) {constB=10}, Alpha=OneMinus(None=@n16.VGradient)), Blend=@n15.RGB) {materialFunction=&Blend_Overlay}
  n18 = TextureObjectParameter() {
   group=PlayerIcon
   parameterName=PlayerIcon
   texture=/Game/UI/Foundation/Textures/Icons/Heroes/Athena/Soldier/T-Soldier-HID-005-Athena-Commando-M
  }
  n19 = TextureSample(
   UVs=MaterialFunctionCall(TextureSize=TextureProperty(Tex=@n18)) {materialFunction=&MF_UI_UVScaling}.ScaleY
   Tex=@n18
  ) {samplerSource=SSM_Clamp_WorldGroupSettings}
  n20 = LinearInterpolate(A=StaticSwitchParameter(True=Saturate(None=@n17.Result), False=@n15.RGB) {parameterName=Bg Gradient}, B=@n19.RGB, Alpha=@n19.RGB)
  NamedRerouteDeclaration(None=StaticSwitchParameter(True=Desaturation(None=@n20), False=@n20) {parameterName=Inactive}) {name=Avatar, nodeColor=<rgba 0 0 0 1>}
  n21 = NamedRerouteUsage()
  n22 = MaterialFunctionCall(UVs=TextureCoordinate(), "Rotation Angle (0-1)"=Constant() {r=0.125}) {materialFunction=&CustomRotator}
  n23 = Constant() {r=0.68}
  n24 = Constant() {r=-0.05}
  n25 = Constant() {r=0.004}
  n26 = Saturate(None=MaterialFunctionCall(
    DistanceField=MaterialFunctionCall(UVs=@n22.Rotated Values, Size=@n23, CornerRoundness=@n24, OutlineThickness=@n25) {materialFunction=&MF_UI_Box}.SDFOutline
    Threshold=Constant() {r=0.07}
   ) {materialFunction=&MF_UI_AntiAliasing}.Inner)
  n27 = NamedRerouteUsage()
  n28 = ComponentMask(None=Subtract(A=NamedRerouteUsage(), B=@n26)) {r=true}
  NamedRerouteDeclaration(None=Add(A=LinearInterpolate(B=@n21, Alpha=@n26), B=Multiply(A=@n27, B=@n28))) {name=DiamondBorder, nodeColor=<rgba 0 0 0 1>}
  n29 = Constant() {r=1}
  n30 = Constant() {r=1}
  n31 = MaterialFunctionCall(
   DistanceField=MaterialFunctionCall(UVs=MaterialFunctionCall() {materialFunction=&GetUserInterfaceUV}.Normalized UV, Size=Subtract(A=@n30)) {materialFunction=&MF_UI_Circle}.SDF
   Threshold=Constant() {r=1}
  ) {materialFunction=&MF_UI_AntiAliasing}
  NamedRerouteDeclaration(None=Multiply(A=Saturate(None=@n29), B=Saturate(None=@n31.Inner))) {name=CircleAlpha, nodeColor=<rgba 0 0 0 1>}
  Reroute(None=@n15.RGB)
  StaticSwitchParameter(True=NamedRerouteUsage(), False=StaticSwitchParameter(True=NamedRerouteUsage(), False=NamedRerouteUsage()) {parameterName=CirlceMask}) {parameterName=DiamondMask}
  n32 = NamedRerouteUsage()
  n33 = StaticSwitchParameter(True=NamedRerouteUsage(), False=NamedRerouteUsage()) {defaultValue=true, parameterName=Square Border}
  StaticSwitchParameter(
   True=NamedRerouteUsage()
   False=StaticSwitchParameter(True=StaticSwitchParameter(True=@n32, False=@n33) {parameterName=CircleBorder}, False=NamedRerouteUsage()) {defaultValue=true, parameterName=White Border}
  ) {parameterName=DiamondBorder}
  Switch!(SwitchValue, Default, None)
 ]
}
