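# Translucent UI master material for Sparks UI backgrounds, drawing a rounded box with a configurable gradient fill.
# MF_UI_Box supplies fill and outline masks, with uniform or per-corner roundness chosen by static switch.
# The two-color gradient feeds an optional radial bottom glow, a Selected blend and a switchable playlist-extension region, then gets soft-light vector noise.
# Named reroutes carry shapes and colors, including a second box mask for the extension's rounded corners.
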Material /SparksUICommon/Materials/M_UI_Sparks_BackgroundMasterM
graph {
 properties {
  bAutomaticallySetUsageInEditor=false
  bUsedWithStaticMesh=false
  blendMode=BLEND_Translucent
  materialDomain=MD_UI
  usedShadingModels=Default Lit
 }
 nodes [
  Constant() {r=1}
  n1 = ScalarParameter() {defaultValue=0.5, parameterName=ExtensionCuttoff}
  n2 = SmoothStep(
   Min=@n1
   Max=@n1
   Value=MaterialFunctionCall() {materialFunction=&LinearGradient}.UGradient
  )
  LinearInterpolate(A=@n2, B=OneMinus(None=NamedRerouteUsage()), Alpha=ScalarParameter() {parameterName=b_HasRoundedInternalCorners})
  n3 = TextureCoordinate()
  n4 = Constant() {r=0.125}
  n5 = Constant() {r=0.2}
  n6 = Constant() {r=1.4}
  n7 = LinearInterpolate(
   A=Constant3Vector() {constant=<rgba 1 1 1 1>}
   B=Constant3Vector() {constant=<rgba 1 0.189 0.2272 1>}
   Alpha=MaterialFunctionCall(
    UVs=MaterialFunctionCall(UVs=@n3, "Rotation Angle (0-1)"=@n4) {materialFunction=&CustomRotator}.Rotated Values
    Size=AppendVector(A=@n5, B=@n6)
    GlowMin=Constant() {r=0.5}
   ) {materialFunction=&MF_UI_Box}.Fill
  )
  n8 = MaterialFunctionCall() {materialFunction=&LinearGradient}
  n9 = LinearInterpolate(A=SmoothStep(Value=@n8.UGradient) {constMin=0.3}, B=@n8.VGradient, Alpha=ScalarParameter() {defaultValue=1, parameterName=GradientPosition})
  n10 = OneMinus(None=@n9)
  n11 = ScalarParameter() {parameterName=GradientInvert}
  n12 = VectorParameter() {parameterName=Background_GR2, defaultValue=<rgba 0.2147 1 0.1704 1>}
  n13 = LinearInterpolate(
   A=LinearInterpolate(
    A=@n12.RGB
    B=VectorParameter() {parameterName=Background_GR1, defaultValue=<rgba 1 0 0.8212 1>}.RGB
    Alpha=LinearInterpolate(A=@n9, B=@n10, Alpha=@n11)
   )
   B=@n12.RGB
   Alpha=ScalarParameter() {parameterName=Greadient}
  )
  LinearInterpolate(
   A=@n13
   B=LinearInterpolate(
    A=@n13
    B=MaterialFunctionCall(Base=@n7, Blend=@n13) {materialFunction=&Blend_Overlay}.Result
    Alpha=NamedRerouteUsage()
   )
   Alpha=ScalarParameter() {defaultValue=1, parameterName=OutlineOP}
  )
  n14 = MaterialFunctionCall() {materialFunction=&BreakOutFloat2Components}
  n15 = NamedRerouteDeclaration(None=@n14.R) {name=PixelSize.X, nodeColor=<rgba 0 0 0 1>}
  Max(A=@n15) {constB=0.5}
  n16 = NamedRerouteDeclaration(None=@n14.G) {name=PixelSize.Y, nodeColor=<rgba 0 0 0 1>}
  Multiply(A=@n16) {constB=0.5}
  n17 = ScalarParameter() {parameterName=0}
  n18 = ScalarParameter() {parameterName=1}
  n19 = ScalarParameter() {parameterName=2}
  n20 = ScalarParameter() {parameterName=3}
  n21 = MaterialFunctionCall(
   Size=Constant() {r=1}
   CornerRoundness=StaticSwitchParameter(
    True=MaterialFunctionCall(R=@n17, G=@n18, B=@n19, A=@n20) {materialFunction=&AppendMany}.RGBA
    False=ScalarParameter() {defaultValue=0.25, parameterName=CornerRoudness}
   ) {parameterName=SpecificCorners}
   OutlineThickness=ScalarParameter() {defaultValue=0.05, parameterName=OutlineThickness}
   KeepAspectRatio=StaticBool() {value=true}
  ) {materialFunction=&MF_UI_Box}
  NamedRerouteDeclaration(None=LinearInterpolate(B=@n21.Fill, Alpha=ScalarParameter() {defaultValue=1, parameterName=MainShapeOp})) {name=MainShape, nodeColor=<rgba 0 0.9766 1 1>}
  NamedRerouteDeclaration(None=@n21.Outline) {name=OutlineShape, nodeColor=<rgba 1 0.7031 0 1>}
  NamedRerouteDeclaration(None=Divide(A=@n15, B=@n16)) {name=Ratio, nodeColor=<rgba 0 0 0 1>}
  n22 = VectorParameter() {parameterName=CL_CE2, defaultValue=<rgba 0.0594 0.0141 0.3542 1>}
  n23 = VectorParameter() {parameterName=CL_CE, defaultValue=<rgba 0.0332 0.0753 0.276 1>}
  n24 = MaterialFunctionCall() {materialFunction=&LinearGradient}
  n25 = VectorParameter() {parameterName=BottomGlow, defaultValue=<rgba 0 0.0312 1 1>}
  n26 = TextureCoordinate()
  n27 = AppendVector(A=Constant() {r=2}, B=Constant() {r=0.8})
  n28 = Constant()
  n29 = Constant() {r=-0.89}
  n30 = MaterialFunctionCall(UV=Add(
    A=MaterialFunctionCall(UVs=@n26, Texture Scale=@n27) {materialFunction=&ScaleUVsByCenter}.UVs
    B=AppendVector(A=@n28, B=@n29)
   )) {materialFunction=&MF_UI_RadialGradient}
  n31 = LinearInterpolate(
   A=StaticSwitchParameter(True=LinearInterpolate(A=@n13, B=@n25.RGB, Alpha=@n30.Smooth), False=@n13) {parameterName=BottomGlow}
   B=NamedRerouteUsage()
   Alpha=ScalarParameter() {parameterName=Selected}
  )
  n32 = LinearInterpolate(
   A=@n31
   B=LinearInterpolate(A=@n31, B=LinearInterpolate(A=@n22.RGB, B=@n23.RGB, Alpha=@n24.VGradient), Alpha=@n2)
   Alpha=ScalarParameter() {defaultValue=1, parameterName=b_PlayList Extension}
  )
  n33 = VectorNoise()
  n34 = ScalarParameter() {defaultValue=0.1, parameterName=NoiseAmount}
  NamedRerouteDeclaration(None=MaterialFunctionCall!(Base=StaticSwitchParameter(True=@n32, False=@n31) {parameterName=PlayList Extension}, Blend=Multiply(A=@n33, B=@n34) {constB=0.1}) {materialFunction=/Engine/Functions/Engine_MaterialFunctions03/Blends/Blend_SoftLight}.Result) {name=Colour, nodeColor=<rgba 0.3828 0 1 1>}
  n35 = MaterialFunctionCall() {materialFunction=&LinearGradient}
  NamedRerouteDeclaration(None=LinearInterpolate(
    A=Constant3Vector() {constant=<rgba 1 1 1 1>}
    B=Constant3Vector() {constant=<rgba 0.2196 0.6792 0.9167 1>}
    Alpha=SmoothStep(Value=@n35.VGradient) {constMin=0.5}
   )) {name=Selected, nodeColor=<rgba 0.1686 0.238 1 1>}
  n36 = Add(A=TextureCoordinate(), B=AppendVector(A=ScalarParameter() {defaultValue=0.25, parameterName=ExtensionCuttoffRoundedCorners}, B=Constant()))
  n37 = AppendVector(A=Constant() {r=0.5}, B=Constant() {r=1})
  n38 = LinearInterpolate(B=ScalarParameter() {defaultValue=0.184, parameterName=3_3}, Alpha=ScalarParameter() {defaultValue=1, parameterName=AnimSpecificCorners})
  n39 = MaterialFunctionCall(R=ScalarParameter() {parameterName=0_1}, G=ScalarParameter() {parameterName=1_1}, B=@n38, A=@n38) {materialFunction=&AppendMany}
  n40 = StaticBool() {value=true}
  NamedRerouteDeclaration(None=Saturate(None=MaterialFunctionCall(UVs=@n36, Size=@n37, CornerRoundness=@n39.RGBA, KeepAspectRatio=@n40) {materialFunction=&MF_UI_Box}.Fill)) {name=FillPlayListExtension, nodeColor=<rgba 1 0.0234 0 1>}
  NamedRerouteUsage()
  NamedRerouteUsage()
  Reroute(None=NamedRerouteUsage())
 ]
}
