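# Masked UI material for album art in the Sparks curated set list, sampling a T_AlbumArt texture parameter.
# GetUserInterfaceUV supplies normalized UV and pixel size, and a "squaredHeight" aspect-ratio factor scales the coordinates used by the texture.
# The same coordinates feed MF_UI_Box, which takes separate roundness parameters for each corner.
# A Padding parameter, aspect-corrected by pixel height over width, feeds a subtraction.
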Material /SparksUICommon/Materials/M_UI_Sparks_CuratedSetList_AlbumArt
graph {
 properties {blendMode=BLEND_Masked, materialDomain=MD_UI, pixelDepthOffsetMode=PDOM_Legacy, usedShadingModels=Default Lit}
 nodes [
  Constant() {r=1}
  n1 = Divide(A=NamedRerouteUsage(), B=Max(A=NamedRerouteUsage()) {constB=0.001})
  n2 = Constant() {r=1}
  n3 = Multiply(A=NamedRerouteUsage(), B=NamedRerouteDeclaration(None=AppendVector(A=@n1, B=@n2)) {name=squaredHeight, nodeColor=<rgba 0 0 0 1>})
  MaterialFunctionCall(
   UVs=@n3
   Size=Constant() {r=1.01}
   CornerRoundness=MaterialFunctionCall(
    X=ScalarParameter() {defaultValue=0.25, parameterName=Corner_TopLeft}
    Y=ScalarParameter() {defaultValue=0.25, parameterName=Corner_BottomLeft}
    Z=ScalarParameter() {defaultValue=-0.1, parameterName=Corner_TopRight}
    A=ScalarParameter() {defaultValue=-0.1, parameterName=Corner_BottomRight}
   ) {materialFunction=&MakeFloat4}.Result
  ) {materialFunction=&MF_UI_Box}
  n4 = MaterialFunctionCall() {materialFunction=&GetUserInterfaceUV}
  NamedRerouteDeclaration(None=@n4.Normalized UV) {name=NormalizedUV, nodeColor=<rgba 0 0 0 1>}
  StaticBool()
  n5 = MaterialFunctionCall(Float2=@n4.Pixel Size) {materialFunction=&BreakOutFloat2Components}
  n6 = Divide(
   A=NamedRerouteDeclaration(None=@n5.G) {name=PixelSize.Y, nodeColor=<rgba 0 0 0 1>}
   B=Max(A=NamedRerouteDeclaration(None=@n5.R) {name=PixelSize.X, nodeColor=<rgba 0 0 0 1>}) {constB=0.001}
  )
  n7 = ScalarParameter() {parameterName=Padding}
  Subtract(B=AppendVector(A=Multiply(A=@n6, B=@n7), B=@n7))
  TextureSampleParameter2D(UVs=@n3) {parameterName=T_AlbumArt}
 ]
}
