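# Translucent UI material that draws a Sparks fullscreen gradient background with optional animation, drifting triangles and grain.
# Blends a two-color linear gradient with left and right radial gradients warped by panning noise, using sine time to oscillate their shapes.
# Two scrolling particle grids build a masked triangle layer, color-dodged over the gradient.
# Static switches select single color, simple gradient, triangles, animation, soft-light noise and rounded-box alpha with optional pixel-based radius.
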Material /SparksUICommon/Materials/M_UI_Sparks_Fullscreen_Background
graph {
 properties {blendMode=BLEND_Translucent, materialDomain=MD_UI, pixelDepthOffsetMode=PDOM_Legacy, usedShadingModels=Default Lit}
 nodes [
  n1 = NamedRerouteUsage()
  n2 = NamedRerouteUsage()
  Max(A=@n1, B=@n2)
  n3 = TextureCoordinate()
  n4 = Constant()
  n5 = ScalarParameter() {defaultValue=-0.1, group=Distortion, parameterName=DistortionSpeed}
  n6 = TextureSample!(
   UVs=Panner(
    Coordinate=MaterialFunctionCall!(Vector or UVs=@n3) {materialFunction=/FortUILibrary/MaterialFunctions/Basics/MF_UI_CheaperVectorToRadial}.Radial Coordinates
    Speed=AppendVector(A=@n4, B=@n5)
   ) {speedY=-0.125}
   Tex
   "DDX(UVs)"=NamedRerouteUsage()
   "DDY(UVs)"=NamedRerouteUsage()
   Apply View MipBias
  ) {mipValueMode=TMVM_Derivative, samplerType=SAMPLERTYPE_Grayscale, texture=/FortUILibrary/UIKit/Buttons/Textures/LowResBlurredNoise}
  n7 = Constant() {r=0.333}
  n8 = Constant() {r=2.333}
  n9 = LinearInterpolate(
   A=@n6.R
   Alpha=Multiply(A=MaterialFunctionCall(Radius=@n7, Density=@n8) {materialFunction=&RadialGradientExponential}.RadialGradientExponential)
  )
  n10 = Multiply(A=@n3) {constB=0.5}
  Multiply(
   A=@n9
   B=TextureSample!(UVs=Panner(Coordinate=@n10) {speedX=0.05, speedY=0.1}, Tex, Apply View MipBias) {samplerType=SAMPLERTYPE_Grayscale, texture=&LowResBlurredNoise}.R
  )
  Multiply(
   A=StaticSwitchParameter(True=NamedRerouteUsage(), False=Constant() {r=1}) {parameterName=IsRoundedBox}
   B=ScalarParameter() {defaultValue=1, parameterName=Alpha_Final}
  )
  NamedRerouteDeclaration(None=@n9) {name=DistortionNoise, nodeColor=<rgba 0 0 0 1>}
  NamedRerouteDeclaration(None=DDX(Value=@n3)) {name=DDX, nodeColor=<rgba 1 0 0.5391 1>}
  NamedRerouteDeclaration(None=MaterialFunctionCall() {materialFunction=&MF_UI_UVScaling}.Scaled To Fill) {name=TriUV, nodeColor=<rgba 1 0 0.3281 1>}
  n11 = Constant() {r=0.05}
  n12 = Constant() {r=-0.05}
  n13 = MaterialFunctionCall!(
   UV=Panner(Coordinate=NamedRerouteUsage(), Speed=AppendVector(A=@n11, B=@n12))
   Global Offset
   Global Rotation
   Tiling=ScalarParameter() {defaultValue=5, parameterName=TriSet1_Tile}
   Cell Offset=Constant()
   Cell Scale
   Cell Rotation
   Random Seed
  ) {materialFunction=/FortUILibrary/MaterialFunctions/Basics/MF_UI_ParticleGrid}
  NamedRerouteDeclaration(None=@n13.Random Cell Values) {name=RandomCells, nodeColor=<rgba 1 0.3281 0 1>}
  n14 = MaterialFunctionCall(UVs=TextureCoordinate(), "Rotation Angle (0-1)"=Constant() {r=0.125}) {materialFunction=&CustomRotator}
  n15 = ComponentMask(None=Add(A=@n14.Rotated Values) {constB=-0.25}) {r=true}
  NamedRerouteDeclaration(None=SmoothStep(Value=OneMinus(None=@n15)) {constMin=0.125}) {name=DepthMask, nodeColor=<rgba 0.4766 1 0 1>}
  NamedRerouteDeclaration(None=DDY(Value=@n3)) {name=DDY, nodeColor=<rgba 0.2422 1 0 1>}
  n16 = ComponentMask(None=Sine(None=@n14.Rotated Values) {period=2}) {g=true}
  NamedRerouteDeclaration(None=Saturate(None=OneMinus(None=@n16))) {name=SineFalloff, nodeColor=<rgba 0 0.6719 1 1>}
  n17 = Time()
  n18 = Cosine(None=Multiply(A=@n17) {constB=-0.1})
  n19 = Sine(None=Multiply(A=Add(A=@n17) {constB=0.2}) {constB=0.1})
  n20 = NamedRerouteUsage()
  n21 = Cosine(None=@n20)
  n22 = LinearInterpolate(Alpha=@n20) {constA=-1}
  n23 = Sine(None=@n20)
  n24 = SmoothStep(
   Max=NamedRerouteUsage()
   Value=MaterialFunctionCall!(
    UVs=@n13.Cells UV
    Corner 01=AppendVector(A=@n18, B=@n19)
    Corner 02=AppendVector(A=@n21, B=@n22)
    Corner 03=AppendVector(A=@n22, B=@n23)
    Glow Min
    Glow Max
    Outline Thickness
   ) {materialFunction=/FortUILibrary/MaterialFunctions/Shapes/MF_UI_WIP_Triangle}.SDF
  )
  n25 = Multiply(A=SmoothStep(Min=NamedRerouteUsage(), Value=OneMinus(None=@n24)) {constMin=0.5}, B=SmoothStep(Value=NamedRerouteUsage()))
  n26 = Constant() {r=0.025}
  n27 = Constant() {r=-0.025}
  n28 = MaterialFunctionCall(
   UV=Panner(Coordinate=NamedRerouteUsage(), Speed=AppendVector(A=@n26, B=@n27))
   Tiling=ScalarParameter() {defaultValue=3, parameterName=TriSet2_Tile}
   Cell Offset=Constant()
  ) {materialFunction=&MF_UI_ParticleGrid}
  n29 = Time()
  n30 = Cosine(None=Multiply(A=@n29) {constB=-0.005})
  n31 = Sine(None=Multiply(A=Add(A=@n29) {constB=0.2}) {constB=0.005})
  n32 = NamedRerouteDeclaration(None=@n28.Random Cell Values) {name=RandomCells_2, nodeColor=<rgba 1 0.3281 0 1>}
  n33 = Cosine(None=@n32)
  n34 = LinearInterpolate(Alpha=@n32) {constA=-1}
  n35 = Sine(None=@n32)
  n36 = SmoothStep(
   Max=NamedRerouteUsage()
   Value=MaterialFunctionCall(UVs=@n28.Cells UV, Corner 01=AppendVector(A=@n30, B=@n31), Corner 02=AppendVector(A=@n33, B=@n34), Corner 03=AppendVector(A=@n34, B=@n35)) {materialFunction=&MF_UI_WIP_Triangle}.SDF
  )
  n37 = Multiply(A=SmoothStep(Min=NamedRerouteUsage(), Value=OneMinus(None=@n36)) {constMin=0.5}, B=SmoothStep(Value=NamedRerouteUsage()))
  NamedRerouteDeclaration(None=Multiply(
    A=MaterialFunctionCall("Value (0-1)"=@n15) {materialFunction=&MF_UI_EaseCurves}.EaseInCubic
    B=Add(A=@n25, B=@n37)
   )) {name=TrianglesMask, nodeColor=<rgba 0 1 0.0859 1>}
  n38 = NamedRerouteUsage()
  n39 = ScalarParameter() {defaultValue=0.05, parameterName=DistrotionAmount}
  n40 = ScalarParameter() {defaultValue=0.7, group=Radial Gradient Left, parameterName=LeftGradient_Radius}
  n41 = ScalarParameter() {defaultValue=0.05, group=Radial Gradient Left, parameterName=LeftGradient_Radius_Adjustment}
  n42 = Add(A=@n40, B=@n41) {constB=0.05}
  n43 = Add(A=@n40, B=Multiply(A=@n41) {constB=-1}) {constB=-0.05}
  n44 = MaterialFunctionCall(Speed=ScalarParameter() {defaultValue=0.03, group=Radial Gradient Left, parameterName=LeftGradient_RadiusSpeed}) {materialFunction=&DebugTimeSine}
  n45 = ScalarParameter() {defaultValue=2.333, group=Radial Gradient Left, parameterName=LeftGradient_Density}
  n46 = ScalarParameter() {defaultValue=0.05, group=Radial Gradient Left, parameterName=LeftGradient_Density_Adjustment}
  n47 = Add(A=@n45, B=@n46) {constB=0.05}
  n48 = Add(A=@n45, B=Multiply(A=@n46) {constB=-1}) {constB=-0.05}
  n49 = MaterialFunctionCall(Speed=ScalarParameter() {defaultValue=0.0125, group=Radial Gradient Left, parameterName=LeftGradient_DensitySpeed}) {materialFunction=&DebugTimeSine}
  n50 = MaterialFunctionCall(
   UVs=Add(A=MaterialFunctionCall() {materialFunction=&MF_UI_UVScaling}.ScaleX, B=Multiply(A=@n38, B=@n39))
   CenterPosition=AppendVector(A=Constant(), B=Constant() {r=1})
   Radius=StaticSwitchParameter(
    True=LinearInterpolate(A=@n42, B=@n43, Alpha=@n44.Result)
    False=ScalarParameter() {defaultValue=0.7, group=Radial Gradient Left, parameterName=LeftGradient_Radius}
   ) {defaultValue=true, parameterName=IsAnimatedBG}
   Density=StaticSwitchParameter(
    True=LinearInterpolate(A=@n47, B=@n48, Alpha=@n49.Result)
    False=ScalarParameter() {defaultValue=2.333, group=Radial Gradient Left, parameterName=LeftGradient_Density}
   ) {defaultValue=true, parameterName=IsAnimatedBG}
  ) {materialFunction=&RadialGradientExponential}
  n51 = ScalarParameter() {defaultValue=1, group=Radial Gradient Left, parameterName=LeftGradient_FalloffRate}
  NamedRerouteDeclaration(None=Saturate(None=Power(Base=@n50.RadialGradientExponential, Exp=@n51) {constExponent=1})) {name=LeftGradient_Mask, nodeColor=<rgba 0 0 0 1>}
  n52 = NamedRerouteUsage()
  n53 = ScalarParameter() {defaultValue=0.05, parameterName=DistrotionAmount}
  n54 = ScalarParameter() {defaultValue=0.6, group=Radial Gradient Right, parameterName=RightGradient_Radius}
  n55 = ScalarParameter() {defaultValue=0.05, group=Radial Gradient Right, parameterName=RightGradient_Radius_Adjustment}
  n56 = Add(A=@n54, B=@n55) {constB=0.05}
  n57 = Add(A=@n54, B=Multiply(A=@n55) {constB=-1}) {constB=-0.05}
  n58 = MaterialFunctionCall(Speed=ScalarParameter() {defaultValue=0.075, group=Radial Gradient Right, parameterName=RightGradient_RadiusSpeed}) {materialFunction=&DebugTimeSine}
  n59 = ScalarParameter() {defaultValue=2.333, group=Radial Gradient Right, parameterName=RightGradient_Density}
  n60 = ScalarParameter() {defaultValue=0.05, group=Radial Gradient Right, parameterName=RightGradient_Density_Adjustment}
  n61 = Add(A=@n59, B=@n60) {constB=0.05}
  n62 = Add(A=@n59, B=Multiply(A=@n60) {constB=-1}) {constB=-0.05}
  n63 = MaterialFunctionCall(Speed=ScalarParameter() {defaultValue=0.05, group=Radial Gradient Right, parameterName=RightGradient_DensitySpeed}) {materialFunction=&DebugTimeSine}
  n64 = MaterialFunctionCall(
   UVs=Add(A=MaterialFunctionCall() {materialFunction=&MF_UI_UVScaling}.ScaleX, B=Multiply(A=@n52, B=@n53))
   CenterPosition=AppendVector(A=Constant() {r=1}, B=Constant())
   Radius=StaticSwitchParameter(
    True=LinearInterpolate(A=@n56, B=@n57, Alpha=@n58.Result)
    False=ScalarParameter() {defaultValue=0.6, group=Radial Gradient Right, parameterName=RightGradient_Radius}
   ) {defaultValue=true, parameterName=IsAnimatedBG}
   Density=StaticSwitchParameter(
    True=LinearInterpolate(A=@n61, B=@n62, Alpha=@n63.Result)
    False=ScalarParameter() {defaultValue=2.333, group=Radial Gradient Right, parameterName=RightGradient_Density}
   ) {defaultValue=true, parameterName=IsAnimatedBG}
  ) {materialFunction=&RadialGradientExponential}
  n65 = ScalarParameter() {defaultValue=1, group=Radial Gradient Right, parameterName=RightGradient_FalloffRate}
  NamedRerouteDeclaration(None=Saturate(None=Power(Base=@n64.RadialGradientExponential, Exp=@n65) {constExponent=1})) {name=RightGradient_Mask, nodeColor=<rgba 0 0 0 1>}
  n66 = MaterialFunctionCall(Float2=MaterialFunctionCall() {materialFunction=&GetUserInterfaceUV}.Pixel Size) {materialFunction=&BreakOutFloat2Components}
  n67 = NamedRerouteDeclaration(None=@n66.R) {name=PixelX, nodeColor=<rgba 1 0.0469 0 1>}
  n68 = NamedRerouteDeclaration(None=@n66.G) {name=PixelY, nodeColor=<rgba 0 0.1562 1 1>}
  n69 = Max(A=@n67, B=@n68)
  NamedRerouteDeclaration(None=Divide(A=Min(A=@n67, B=@n68), B=Max(A=@n69) {constB=0.001})) {name=Ratio, nodeColor=<rgba 0.3125 1 0 1>}
  n70 = Constant() {r=1}
  n71 = ScalarParameter() {parameterName=Padding}
  n72 = Multiply(A=@n71, B=NamedRerouteUsage())
  n73 = LinearInterpolate(A=AppendVector(A=@n71, B=@n72), B=AppendVector(A=@n72, B=@n71), Alpha=NamedRerouteUsage())
  n74 = Divide(
   A=Multiply(A=ScalarParameter() {defaultValue=16, parameterName=CornerRadiusPixels}) {constA=-0.9746, constB=2}
   B=Max(A=Min(A=@n1, B=@n2)) {constB=0.001}
  )
  n75 = ScalarParameter() {defaultValue=0.5, parameterName=Corner_Radius}
  NamedRerouteDeclaration(None=MaterialFunctionCall(
    Size=Subtract(A=@n70, B=@n73)
    CornerRoundness=StaticSwitchParameter(True=@n74, False=@n75) {parameterName=PixelRadius}
    KeepAspectRatio=StaticBool() {value=true}
   ) {materialFunction=&MF_UI_Box}.Fill) {name=Box Mask, nodeColor=<rgba 0 0 0 1>}
  NamedRerouteDeclaration(None=MaterialFunctionCall!(A=NamedRerouteUsage(), B=NamedRerouteUsage()) {materialFunction=/FortUILibrary/MaterialFunctions/Basics/MF_UI_BooleanOperations}.IsGreater) {name=YisLarger, nodeColor=<rgba 0.2422 0 1 1>}
  n76 = Multiply(A=NamedRerouteUsage(), B=ScalarParameter() {defaultValue=1, parameterName=TriangleIntensity})
  n77 = ScalarParameter() {defaultValue=-0.1, group=LinearGradient, parameterName=LinearGradientStretch}
  n78 = ScalarParameter() {defaultValue=0.1, group=LinearGradient, parameterName=StretchStartAmount}
  n79 = Multiply(A=@n78) {constB=-1}
  n80 = ScalarParameter() {defaultValue=0.1, group=LinearGradient, parameterName=GradientStretch_OscillationSpeed}
  n81 = StaticSwitchParameter(
   True=LinearInterpolate(
    A=Add(A=@n77, B=@n78) {constB=0.1}
    B=Add(A=@n77, B=@n79) {constB=-0.1}
    Alpha=MaterialFunctionCall(Speed=@n80) {materialFunction=&DebugTimeSine}.Result
   )
   False=ScalarParameter() {defaultValue=-0.1, group=LinearGradient, parameterName=LinearGradientStretch}
  ) {defaultValue=true, parameterName=IsAnimatedBG}
  n82 = ScalarParameter() {defaultValue=1, group=LinearGradient, parameterName=LinearGradientStretch_End}
  n83 = ScalarParameter() {defaultValue=0.1, group=LinearGradient, parameterName=StretchEndAmount}
  n84 = Multiply(A=@n83) {constB=-1}
  n85 = ScalarParameter() {defaultValue=0.1, group=LinearGradient, parameterName=GradientStretch_OscillationSpeed}
  n86 = StaticSwitchParameter(
   True=LinearInterpolate(
    A=Add(A=@n82, B=@n83) {constB=0.1}
    B=Add(A=@n82, B=@n84) {constB=-0.1}
    Alpha=MaterialFunctionCall(Speed=@n85) {materialFunction=&DebugTimeSine}.Result
   )
   False=ScalarParameter() {defaultValue=1, group=LinearGradient, parameterName=LinearGradientStretch_End}
  ) {defaultValue=true, parameterName=IsAnimatedBG}
  n87 = ScalarParameter() {defaultValue=-0.075, group=LinearGradient, parameterName=LinearGradientRotation}
  n88 = ScalarParameter() {defaultValue=0.125, group=LinearGradient, parameterName=GradientRotation_OscillationSpeed}
  n89 = StaticSwitchParameter(
   True=LinearInterpolate(
    A=Add(A=@n87) {constB=0.05}
    B=Add(A=@n87) {constB=-0.05}
    Alpha=MaterialFunctionCall(Speed=@n88) {materialFunction=&DebugTimeSine}.Result
   )
   False=ScalarParameter() {defaultValue=-0.075, group=LinearGradient, parameterName=LinearGradientRotation}
  ) {defaultValue=true, parameterName=IsAnimatedBG}
  n90 = LinearInterpolate(
   A=VectorParameter() {parameterName=Color_LinearGradient_Left, defaultValue=<rgba 1 0 0 1>}.RGB
   B=VectorParameter() {parameterName=Color_LinearGradient_Right, defaultValue=<rgba 0 1 0 1>}.RGB
   Alpha=SmoothStep(
    Max=ScalarParameter() {defaultValue=1, group=LinearGradient, parameterName=GradientSidePreference}
    Value=MaterialFunctionCall(GradientStart=@n81, GradientEnd=@n86, RotationAngle=@n89) {materialFunction=&MF_UI_LinearGradient}.Result
   ) {constMax=0.75}
  )
  n91 = LinearInterpolate(
   A=LinearInterpolate(
    A=NamedRerouteDeclaration(None=@n90) {name=LinearGradientColor, nodeColor=<rgba 0.8047 1 0 1>}
    B=VectorParameter() {parameterName=Color_RadialGradient_Left, defaultValue=<rgba 0 0 1 1>}.RGB
    Alpha=NamedRerouteUsage()
   )
   B=VectorParameter() {parameterName=Color_RadialGradient_Right, defaultValue=<rgba 1 0 1 1>}.RGB
   Alpha=NamedRerouteUsage()
  )
  n92 = Saturate(None=MaterialFunctionCall!(Blend=Saturate(None=@n76), Base=@n91) {materialFunction=/Engine/Functions/Engine_MaterialFunctions03/Blends/Blend_ColorDodge}.Result)
  n93 = StaticSwitchParameter(
   True=VectorParameter() {parameterName=Color, defaultValue=<rgba 1 0 0 1>}.RGB
   False=StaticSwitchParameter(True=NamedRerouteUsage(), False=StaticSwitchParameter(True=@n92, False=@n91) {parameterName=HasTriangles}) {parameterName=IsSimpleLinearGradient}
  ) {parameterName=IsSingleColor}
  n94 = VectorNoise()
  n95 = ScalarParameter() {defaultValue=0.1, parameterName=NoiseAmount}
  StaticSwitchParameter(
   True=MaterialFunctionCall!(Base=@n93, Blend=Multiply(A=@n94, B=@n95) {constB=0.1}) {materialFunction=/Engine/Functions/Engine_MaterialFunctions03/Blends/Blend_SoftLight}.Result
   False=@n93
  ) {parameterName=UseNoise}
 ]
}
