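# Translucent UI base material for Sparks list entry backings, drawing a rounded box with an optional angled edge.
# An aspect-corrected SDF box gives antialiased or soft-edged opacity, and bIsOutline blends shape and color to an outline with glow.
# Static switches select color features: two-color gradient, unique edge color, or cosine gradient ramp with time-blended, panning triangle patterns.
# Anim_ scalars drive the added-song color switch and ramp fade.
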Material /SparksUICommon/Materials/M_UI_Sparks_ListEntry_Backing_Base
graph {
 properties {blendMode=BLEND_Translucent, materialDomain=MD_UI, pixelDepthOffsetMode=PDOM_Legacy, usedShadingModels=Default Lit}
 nodes [
  n1 = VectorParameter() {parameterName=BaseColor, defaultValue=<rgba 0 0.0137 0.063 1>}
  n2 = VectorParameter() {parameterName=RGB1, defaultValue=<rgba 1 0 0 1>}
  n3 = ScalarParameter() {parameterName=Anim_SwitchAddedSong}
  n4 = StaticSwitchParameter(True=NamedRerouteUsage(), False=NamedRerouteUsage()) {parameterName=ColourGradientRamp}
  n5 = NamedRerouteUsage()
  n6 = ScalarParameter() {parameterName=RampGradientTrianglesMixer}
  n7 = StaticSwitchParameter(
   True=LinearInterpolate(A=LinearInterpolate(A=@n1.RGB, B=@n2.RGB, Alpha=@n3) {constAlpha=1}, B=LinearInterpolate(A=@n4, B=@n5, Alpha=@n6), Alpha=@n3)
   False=@n2.RGB
  ) {parameterName=bIs2ColorGradient}
  LinearInterpolate(
   A=StaticSwitchParameter(True=NamedRerouteUsage(), False=NamedRerouteDeclaration(None=@n7) {name=ColorControl, nodeColor=<rgba 0 1 0.1562 1>}) {defaultValue=true, parameterName=bHasUniqueEdgeColor}
   B=NamedRerouteUsage()
   Alpha=ScalarParameter() {parameterName=bIsOutline}
  )
  n8 = MaterialFunctionCall(Float2=NamedRerouteUsage()) {materialFunction=&BreakOutFloat2Components}
  n9 = MaterialFunctionCall(X=@n8.R, Y=@n8.R) {materialFunction=&MakeFloat2}
  n10 = MaterialFunctionCall(X=OneMinus(None=@n8.R), Y=@n8.G) {materialFunction=&MakeFloat2}
  n11 = ScalarParameter() {parameterName=XFlip}
  n12 = Rotator(
   Coordinate=Multiply(A=LinearInterpolate(A=@n9.Result, B=@n10.Result, Alpha=@n11), B=NamedRerouteUsage())
   Time=ScalarParameter() {defaultValue=-0.125, parameterName=EdgeAngle}
  ) {speed=6.2832}
  n13 = AppendVector(A=ScalarParameter() {parameterName=EdgeXOffset}, B=ScalarParameter() {parameterName=EdgeYOffset})
  n14 = Subtract(A=ComponentMask(None=Add(A=@n12, B=@n13)) {r=true}, B=ScalarParameter() {defaultValue=128, parameterName=EdgeOffset})
  n15 = ScalarParameter() {defaultValue=0.5, parameterName=EdgeFallOff}
  NamedRerouteDeclaration(None=Saturate(None=Multiply(A=@n14, B=@n15))) {name=Edge, nodeColor=<rgba 0 0 0 1>}
  n16 = NamedRerouteDeclaration(None=MaterialFunctionCall() {materialFunction=&ScreenAlignedUVs}."X 100%, Y 100%") {name=ScreenSizeUV, nodeColor=<rgba 0 0 0 1>}
  n17 = AppendVector(A=Constant() {r=4}, B=Constant() {r=3})
  n18 = Constant() {r=-0.01}
  n19 = Constant() {r=-0.015}
  n20 = TextureSample(UVs=Panner(Coordinate=Multiply(A=@n16, B=@n17), Speed=AppendVector(A=@n18, B=@n19))) {samplerType=SAMPLERTYPE_LinearColor, texture=/Game/UI/Foundation/Textures/Pattern/T_Triangle_BattlePassSmall}
  n21 = MaterialFunctionCall(
   1=SmoothStep(Value=@n20.R) {constMax=0.2, constMin=0.08}
   2=SmoothStep(Value=@n20.G) {constMax=0.2, constMin=0.08}
   3=SmoothStep(Value=@n20.B) {constMax=0.2, constMin=0.08}
   A=MaterialFunctionCall(Speed=Constant() {r=0.1}) {materialFunction=&DebugTimeSine}.Result
  ) {materialFunction=&Lerp_Multiple_Float3}
  n22 = NamedRerouteUsage()
  n23 = AppendVector(A=Constant() {r=6}, B=Constant() {r=6})
  n24 = Constant() {r=-0.04}
  n25 = Constant() {r=-0.02}
  n26 = TextureSample(UVs=Panner(Coordinate=Multiply(A=@n22, B=@n23), Speed=AppendVector(A=@n24, B=@n25))) {samplerType=SAMPLERTYPE_LinearColor, texture=&T_Triangle_BattlePassSmall}
  n27 = MaterialFunctionCall(
   1=SmoothStep(Value=@n26.R) {constMax=0.2, constMin=0.08}
   2=SmoothStep(Value=@n26.G) {constMax=0.2, constMin=0.08}
   3=SmoothStep(Value=@n26.B) {constMax=0.2, constMin=0.08}
   A=MaterialFunctionCall(Speed=Constant() {r=0.08}) {materialFunction=&DebugTimeSine}.Result
  ) {materialFunction=&Lerp_Multiple_Float3}
  n28 = MaterialFunctionCall(Speed=Constant() {r=0.05}) {materialFunction=&DebugTimeSine}
  NamedRerouteDeclaration(None=Multiply(A=LinearInterpolate(A=@n21.Lerp 3 Inputs, B=@n27.Lerp 3 Inputs, Alpha=@n28.Result))) {name=TriangelePattern, nodeColor=<rgba 0 0.0625 1 1>}
  n29 = Constant3Vector() {constant=<rgba 0.4217 0.1232 1 1>}
  n30 = NamedRerouteUsage()
  n31 = Constant3Vector() {constant=<rgba 0.1062 0.4666 0.5625 1>}
  n32 = NamedRerouteUsage()
  n33 = Saturate(None=LinearInterpolate(A=Min(A=@n29, B=@n30), B=Min(A=@n31, B=@n32)) {constAlpha=0.7})
  n34 = NamedRerouteUsage()
  NamedRerouteDeclaration(None=LinearInterpolate(
    A=MaterialFunctionCall(Base=@n33, Blend=@n34) {materialFunction=&Blend_Overlay}.Result
    B=@n33
   ) {constAlpha=0.1}) {name=ColoredTriangelesPattern, nodeColor=<rgba 1 0.7266 0 1>}
  n35 = Constant() {r=0.1}
  n36 = Constant() {r=0.8}
  n37 = NamedRerouteUsage()
  n38 = Constant()
  n39 = Constant() {r=0.8}
  n40 = NamedRerouteUsage()
  n41 = Constant()
  n42 = Constant() {r=1.2}
  n43 = NamedRerouteUsage()
  n44 = MaterialFunctionCall(
   R=LinearInterpolate(A=LinearInterpolate(A=@n35, B=@n36, Alpha=@n37), B=Constant() {r=-1}, Alpha=ScalarParameter() {parameterName=Anim_GradientRamp_FadeColor})
   G=LinearInterpolate(
    A=LinearInterpolate(A=@n38, B=@n39, Alpha=@n40)
    B=LinearInterpolate(A=@n41, B=@n42, Alpha=@n43)
    Alpha=ScalarParameter() {parameterName=Anim_GradientRamp_FadeColor}
   )
   B=Constant() {r=1}
  ) {materialFunction=&AppendMany}
  n45 = VectorParameter() {parameterName=GradientRampFadeColour, defaultValue=<rgba 0.3249 0.0428 0.7396 1>}
  n46 = Saturate(None=Constant3Vector() {constant=<rgba 0.1395 0.294 0.724 1>})
  n47 = ScalarParameter() {parameterName=Anim_GradientRamp_FadeColor}
  n48 = Saturate(None=LinearInterpolate(
    A=MaterialFunctionCall!(Gradient, a=@n44.RGB, b, c, d) {materialFunction=/FortUILibrary/MaterialFunctions/Color/MF_UI_CosineGradient}.Result
    B=LinearInterpolate(A=@n45.RGB, B=@n46, Alpha=@n47)
    Alpha=LinearInterpolate(Alpha=@n47) {constA=0.5, constB=0.2}
   ))
  NamedRerouteDeclaration(None=@n48) {name=GradientRamp, nodeColor=<rgba 0 1 0.9063 1>}
  NamedRerouteDeclaration(None=@n21.Lerp 3 Inputs) {name=TriangelePattern1, nodeColor=<rgba 0 0.1094 1 1>}
  NamedRerouteDeclaration(None=@n27.Lerp 3 Inputs) {name=TrianglePattern2, nodeColor=<rgba 0.8984 1 0 1>}
  n49 = Saturate(None=NamedRerouteUsage())
  n50 = NamedRerouteUsage()
  n51 = LinearInterpolate(Alpha=MaterialFunctionCall(Speed=Constant() {r=0.05}) {materialFunction=&DebugTimeSine}.Result) {constA=0.7, constB=0.8}
  n52 = MaterialFunctionCall() {materialFunction=&LinearGradient}
  NamedRerouteDeclaration(None=LinearInterpolate(A=Add(A=@n49, B=@n50), B=NamedRerouteUsage(), Alpha=SmoothStep(Max=@n51, Value=@n52.UGradient) {constMax=0.7, constMin=0.2})) {name=RamGradientWithTriangele, nodeColor=<rgba 0.6172 0 1 1>}
  n53 = AppendVector(A=ScalarParameter() {parameterName=XY-Offset-X}, B=ScalarParameter() {parameterName=XY-Offset-Y})
  n54 = MaterialFunctionCall() {materialFunction=&GetUserInterfaceUV}
  n55 = MaterialFunctionCall(Float2=@n54.Pixel Size) {materialFunction=&BreakOutFloat2Components}
  n56 = Divide(
   A=NamedRerouteDeclaration(None=@n55.G) {name=PixelSize.Y, nodeColor=<rgba 0 0 0 1>}
   B=Max(A=NamedRerouteDeclaration(None=@n55.R) {name=PixelSize.X, nodeColor=<rgba 0 0 0 1>}) {constB=0.001}
  )
  n57 = ScalarParameter() {defaultValue=0.05, parameterName=Threshold}
  n58 = AppendVector(A=Multiply(A=Multiply(A=@n56), B=@n57), B=@n57)
  n59 = MaterialFunctionCall(
   Size=Subtract(B=Add(A=@n53, B=@n58))
   CornerRoundness=ScalarParameter() {defaultValue=0.5, parameterName=CornerRadius}
   OutlineThickness=ScalarParameter() {defaultValue=0.05, parameterName=OutlineThicknes}
   KeepAspectRatio=StaticBool() {value=true}
  ) {materialFunction=&MF_UI_Box}
  n60 = Saturate(None=MaterialFunctionCall(DistanceField=@n59.SDF, Threshold=ScalarParameter() {defaultValue=0.05, parameterName=Threshold}) {materialFunction=&MF_UI_AntiAliasing}.Inner)
  n61 = OneMinus(None=SmoothStep(Value=@n59.SDF) {constMax=0, constMin=-0.25})
  n62 = ScalarParameter() {parameterName=FixedSoftEdge}
  n63 = Multiply(
   A=NamedRerouteDeclaration(None=LinearInterpolate(A=@n60, B=@n61, Alpha=@n62)) {name=Box Mask, nodeColor=<rgba 0 0 0 1>}
   B=ScalarParameter() {defaultValue=0.5, parameterName=Opacity}
  )
  n64 = ScalarParameter() {defaultValue=1, parameterName=EdgeOpacity}
  n65 = NamedRerouteUsage()
  n66 = Add(A=@n63, B=Multiply(A=LinearInterpolate(B=@n64, Alpha=@n65), B=NamedRerouteUsage()))
  n67 = NamedRerouteUsage()
  n68 = Power(Base=OneMinus(None=@n67), Exp=ScalarParameter() {parameterName=GlowPower})
  n69 = Add(
   A=OneMinus(None=SmoothStep(Value=@n67) {constMax=0.035})
   B=NamedRerouteDeclaration(None=Saturate(None=@n68)) {name=GlowAlpha, nodeColor=<rgba 0 0 0 1>}
  )
  NamedRerouteDeclaration(None=LinearInterpolate(
    A=StaticSwitchParameter(True=@n66, False=@n63) {defaultValue=true, parameterName=bHasEdge}
    B=Saturate(None=@n69)
    Alpha=ScalarParameter() {parameterName=bIsOutline}
   )) {name=BaseShape, nodeColor=<rgba 1 0.9609 0 1>}
  NamedRerouteDeclaration(None=LinearInterpolate(
    A=VectorParameter() {parameterName=RGB-Glow, defaultValue=<rgba 1 0 0 1>}.RGB
    B=VectorParameter() {parameterName=RGB1, defaultValue=<rgba 1 0 0 1>}.RGB
    Alpha=NamedRerouteUsage()
   )) {name=Outline+GlowColor, nodeColor=<rgba 0 1 0.8828 1>}
  NamedRerouteDeclaration(None=@n54.9-Slice UV) {name=9sliceuv, nodeColor=<rgba 0 0 0 1>}
  n70 = NamedRerouteUsage()
  n71 = ScalarParameter() {defaultValue=1, parameterName=EdgeOpacity}
  NamedRerouteDeclaration(None=LinearInterpolate(A=NamedRerouteUsage(), B=VectorParameter() {parameterName=RGB3, defaultValue=<rgba 0 1 1 1>}.RGB, Alpha=Multiply(A=@n70, B=@n71))) {name=UniqueEdgeColor, nodeColor=<rgba 0.0078 0 1 1>}
  NamedRerouteDeclaration(None=@n54.Normalized UV) {name=NormalizedUV, nodeColor=<rgba 0 0 0 1>}
  n72 = Max(A=@n55.G) {constB=0.001}
  NamedRerouteDeclaration(None=AppendVector(A=Divide(A=@n55.R, B=@n72), B=Constant() {r=1})) {name=squaredHeight, nodeColor=<rgba 0 0 0 1>}
  NamedRerouteDeclaration(None=@n59.SDFOutline) {name=SDF Outline, nodeColor=<rgba 0 0 0 1>}
  NamedRerouteDeclaration(None=MaterialFunctionCall(Speed=Constant() {r=0.1}) {materialFunction=&DebugTimeSine}.Result) {name=PingPongTime, nodeColor=<rgba 0.2891 1 0 1>}
  n73 = Add(A=ComponentMask(None=NamedRerouteUsage()) {r=true}, B=AppendVector(A=ScalarParameter() {parameterName=OffsetGradient_X}, B=Constant()))
  n74 = MaterialFunctionCall("Value (0-1)"=@n73) {materialFunction=&MF_UI_EaseCurves}
  n75 = ScalarParameter() {parameterName=EaseCurve_Amount}
  NamedRerouteDeclaration(None=LinearInterpolate(
    A=VectorParameter() {parameterName=RGB1, defaultValue=<rgba 1 0 0 1>}.RGB
    B=VectorParameter() {parameterName=RGB2, defaultValue=<rgba 1 1 0 1>}.RGB
    Alpha=LinearInterpolate(A=@n73, B=@n74.EaseOutQuint, Alpha=@n75)
   )) {name=ColorGradient, nodeColor=<rgba 0 0.7422 1 1>}
  n76 = MaterialFunctionCall() {materialFunction=&LinearGradient}
  NamedRerouteDeclaration(None=LinearInterpolate(A=@n48, B=NamedRerouteUsage(), Alpha=SmoothStep(Value=@n76.UGradient) {constMax=0.7, constMin=0.1})) {name=RampGradient, nodeColor=<rgba 0 1 0.1094 1>}
  NamedRerouteUsage()
  StaticSwitchParameter(True=StaticSwitchParameter(True=NamedRerouteUsage(), False=NamedRerouteUsage()) {parameterName=RampGradientWithTriangelsPattern}) {parameterName=ColourGradientRamp}
 ]
}
