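# Translucent UI material that draws a glowing rounded outline, named as a list entry highlight for the Sparks UI.
# MF_UI_CoreUI_RoundedMask builds shape and outline SDFs from corner radius, skew, thickness and a padded, aspect-ratio-corrected size.
# A distance field feeds both an anti-aliased mask and an offset, power-shaped glow, which are added together.
# Color blends from blue-tinted glow to a horizontal two-color gradient inside that mask.
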Material /SparksUICommon/Materials/M_UI_Sparks_ListEntry_Hightlight_Outline
graph {
 properties {blendMode=BLEND_Translucent, materialDomain=MD_UI, pixelDepthOffsetMode=PDOM_Legacy, usedShadingModels=Default Lit}
 nodes [
  n1 = NamedRerouteUsage()
  n2 = Multiply(A=Add(A=@n1, B=ScalarParameter() {defaultValue=-0.35, parameterName=Glow Offset}) {constB=-0.05}) {constB=-1}
  n3 = ScalarParameter() {defaultValue=2, parameterName=GlowPower}
  n4 = Saturate(None=Multiply(A=Power(Base=@n2, Exp=@n3), B=ScalarParameter() {defaultValue=1.5, parameterName=GlowMultiply}) {constB=15})
  n5 = ScalarParameter() {defaultValue=0.05, parameterName=Threshold}
  n6 = NamedRerouteDeclaration(None=Saturate(None=MaterialFunctionCall(DistanceField=@n1, Threshold=@n5) {materialFunction=&MF_UI_AntiAliasing}.Inner)) {name=Box Mask, nodeColor=<rgba 0 0 0 1>}
  Add(A=@n4, B=@n6)
  n7 = ScalarParameter() {defaultValue=-1, group=Outline, parameterName=Outline Thickness}
  LinearInterpolate(B=@n7, Alpha=ScalarParameter() {defaultValue=1, group=Anim, parameterName=Anim_OutlineThickness, sliderMax=1})
  n8 = TextureCoordinate()
  LinearInterpolate(
   A=Multiply(A=Constant3Vector() {constant=<rgba 0 0.0933 1 1>}, B=@n4)
   B=LinearInterpolate(
    A=VectorParameter() {parameterName=RGB1, defaultValue=<rgba 0.0844 0.6724 1 1>}.RGB
    B=VectorParameter() {parameterName=RGB2, defaultValue=<rgba 0.6308 0.8963 1 1>}.RGB
    Alpha=ComponentMask(None=@n8) {r=true}
   )
   Alpha=@n6
  )
  n9 = MaterialFunctionCall(
   Size=NamedRerouteUsage()
   Corner Radius=ScalarParameter() {defaultValue=0.117, parameterName=Corner Radius}
   Radius in Pixels=ScalarParameter() {parameterName=Radius in Pixels, sliderMax=1}
   Skew Angle=NamedRerouteDeclaration(None=ScalarParameter() {parameterName=Skew Angle, sliderMax=360}) {name=Skew Angle, nodeColor=<rgba 0 0 0 1>}
   Outline Thickness=NamedRerouteDeclaration(None=@n7) {name=Outline Thickness, nodeColor=<rgba 0 0 0 1>}
  ) {materialFunction=&MF_UI_CoreUI_RoundedMask}
  NamedRerouteDeclaration(None=@n9.Outline Mask) {name=OutlineMask, nodeColor=<rgba 0 0 0 1>}
  NamedRerouteDeclaration(None=@n9.SDF) {name=SDF_Glow, nodeColor=<rgba 0 0 0 1>}
  NamedRerouteDeclaration() {name=Resolution Factor, nodeColor=<rgba 0 0 0 1>}
  n10 = Divide(
   A=NamedRerouteDeclaration(None=ScalarParameter() {defaultValue=4, group=Padding, parameterName=Padding}) {name=Padding, nodeColor=<rgba 0 0 0 1>}
   B=NamedRerouteUsage()
  )
  n11 = NamedRerouteUsage()
  NamedRerouteDeclaration(None=OneMinus(None=Multiply(A=@n10, B=@n11))) {name=Outline Size, nodeColor=<rgba 0 0 0 1>}
  n12 = MaterialFunctionCall() {materialFunction=&GetUserInterfaceUV}
  NamedRerouteDeclaration(None=@n12.Normalized UV) {name=NormalizedUV, nodeColor=<rgba 0 0 0 1>}
  NamedRerouteDeclaration(None=@n9.Shape Mask) {name=ShapeMask, nodeColor=<rgba 0 0 0 1>}
  n13 = NamedRerouteDeclaration(None=@n12.Pixel Size) {name=Pixel Size, nodeColor=<rgba 0 0 0 1>}
  NamedRerouteDeclaration(None=Min(A=ComponentMask(None=@n13) {r=true}, B=ComponentMask(None=@n13) {g=true})) {name=Min Pixel Size, nodeColor=<rgba 0 0 0 1>}
  NamedRerouteDeclaration(None=@n9.Outline SDF) {name=SDFOutline, nodeColor=<rgba 0 0 0 1>}
  n14 = NamedRerouteUsage()
  n15 = MaterialFunctionCall!(Dimensions) {materialFunction=/FortUILibrary/MaterialFunctions/Basics/MF_UI_GetAspectRatio}
  n16 = ComponentMask(None=@n15.Max Ratio) {r=true}
  n17 = ComponentMask(None=@n15.Max Ratio) {g=true}
  n18 = Divide(B=Max(A=@n16, B=@n17))
  n19 = NamedRerouteUsage()
  n20 = Add(A=NamedRerouteUsage(), B=Multiply(A=OneMinus(None=@n18), B=OneMinus(None=@n19)) {constB=0.25})
  NamedRerouteDeclaration(None=LinearInterpolate(
    A=AppendVector(A=@n14, B=@n20)
    B=AppendVector(A=@n20, B=@n14)
    Alpha=MaterialFunctionCall!(A=@n16, B=@n17) {materialFunction=/FortUILibrary/MaterialFunctions/Basics/MF_UI_BooleanOperations}.IsGreater
   )) {name=Size with Padding, nodeColor=<rgba 0 0 0 1>}
 ]
}
